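# Sound cue for the end-of-reload foley of the Figment semi-auto assault rifle, playing the auto rifle single-cock sample.
# A single wave player feeds a Modulator left at its usual settings.
# The cue uses the shared weapon reload attenuation and weapon foley sound class, with reduced volume and overridden concurrency.
# Its debug flag is switched on.
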SoundCue /Figment_Assault_SemiAuto_Athena/Sounds/FNS03_Weapon_ASLT_Auto_Reload_End_Cue
root=Modulator_0
nodes [
 {id=Modulator_0, children [WavePlayer_27]}
 {id=WavePlayer_27, properties {soundWaveAssetPtr=/Game/Sounds/Fort_Weapons/AutoRifle/Reload_Elements/rifle_auto_single_cock}}
]
properties {
 attenuationSettings=&Fort_Weapon_Reload_Att
 bDebug=true
 bOverrideConcurrency=true
 maxDistance=2250
 soundClassObject=&Fort_WeaponFoley
 volumeMultiplier=0.626
}
