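# Sound cue for the equip sound of the Figment semi-auto assault rifle, classed as weapon foley.
# It reuses the Bullpup auto rifle's reload-end wave, played after a short fixed delay at slightly lowered pitch and volume.
# It takes its falloff from the shared weapon reload attenuation, overrides concurrency, and has its debug flag left on.
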SoundCue /Figment_Assault_SemiAuto_Athena/Sounds/SC_FNS03_Weapon_ASLT_Equip
root=Delay_0
nodes [
 {id=Delay_0, children [WavePlayer_1], properties {delayMax=0.1, delayMin=0.1}}
 {id=WavePlayer_1, properties {soundWaveAssetPtr=/Game/Sounds/Fort_Weapons/AutoRifle/Bullpup/Bullpup_Reload_End_01}}
]
properties {
 attenuationSettings=&Fort_Weapon_Reload_Att
 bDebug=true
 bOverrideConcurrency=true
 maxDistance=2250
 pitchMultiplier=0.9
 soundClassObject=&Fort_WeaponFoley
 volumeMultiplier=0.902
}
